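; Results crowd actor for Pilgrim Quickplay that toggles an instanced candy-summer crowd mesh (HISM) with the game's state flow.
; On BeginPlay it waits for the Pilgrim event router, then listens for Quickplay state-ended events.
; It shows the crowd when Outro ends and hides it when SongResults or SetResults ends, propagating visibility to child components.
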(event EventBeginPlay
  (bind _hism_sm_sparks_f_candysummer_crowd_16 (Variables|Default|GetHISM_SM_Sparks_F_CandySummer_Crowd_16))
  ; Listen to Pilgrim Event Quickplay State Event
  (bind _eventrouter (Pilgrim|PilgrimWaitForEventRouter 0)
    (:OnEventRouterAvailable
      (Utilities|IsValid _eventrouter
        (:Is%20Valid
          (bind (_asynctaskproxy _payload) (GameplayEventRouter|StartListeningToEvent _eventrouter "/Script/PilgrimQPSharedRuntime.PilgrimEvent_QuickplayState_Ended")
            (:OnEventReceived
              ; Set Visibility to On after Outro
              (Utilities|FlowControl|Sequence
                (:then_0
                  (if (Utilities|Enum|Equal(Enum) (Utilities|Struct|BreakPilgrimEventQuickplayStateBase _payload) "Outro")
                    (Rendering|SetVisibility _hism_sm_sparks_f_candysummer_crowd_16 true true)))
                (:then_1
                  ; Set Visibility to Off after Results
                  (if (or (Utilities|Enum|Equal(Enum) (Utilities|Struct|BreakPilgrimEventQuickplayStateBase _payload) "SongResults") (Utilities|Enum|Equal(Enum) (Utilities|Struct|BreakPilgrimEventQuickplayStateBase _payload) "SetResults"))
                    (Rendering|SetVisibility _hism_sm_sparks_f_candysummer_crowd_16 false true)))))))))))

(fn ConstructionScript)
